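# Control Rig for the SequinPie backpack cosmetic that procedurally rotates the pack bone in response to the wearer's body pose.
# When SkyDiving is false, two spherical pose readers measure upper-spine (spine_05) orientation against the right lower arm and the pelvis.
# Each reader's output weights its own rotation offset on the pack bone, applied in sequence.
# When SkyDiving is true, no correction runs; the ribbon bones are never driven here.
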ControlRigBlueprint /BRCosmetics/Accessories/FORT_Backpacks/Backpack_M_MED_SequinPie/Meshes/M_MED_SequinPie_Pack_CtrlRig
graphs [
 RigVMFunctionLibrary {links [], nodes []}
 RigVMModel {nodes [
   RigUnit_BeginExecution = &RigUnit_BeginExecution()
   RigVMFunction_ControlFlowBranch = &RigVMFunction_ControlFlowBranch(ExecuteContext=@RigUnit_BeginExecution.ExecutePin, Condition=RigVMVariableNode() {variable=SkyDiving}.Value)
   n1 = &RigUnit_SphericalPoseReader(
    ActiveRegionScaleFactors=(PositiveWidth=1,NegativeWidth=1.5,PositiveHeight=1,NegativeHeight=2)
    ActiveRegionSize=0.01
    Debug=(bDrawDebug=false,bDraw2D=False,bDrawLocalAxes=False,DebugScale=25,DebugSegments=20,DebugThickness=0.25)
    DriverAxis=(X=0,Y=-1,Z=0)
    DriverItem=(Type=Bone,Name="body_jnt_spine_05")
    FalloffRegionScaleFactors=(PositiveWidth=0.4,NegativeWidth=1,PositiveHeight=0.5,NegativeHeight=1)
    FalloffSize=0.1
    FlipHeightScaling=False
    FlipWidthScaling=False
    OptionalParentItem=(Type=Bone,Name="body_jnt_lowerarm_r")
    RotationOffset=(X=-2,Y=52,Z=100)
    ExecutePin=@RigVMFunction_ControlFlowBranch.False
   )
   > &RigUnit_ModifyTransforms(
    ItemToModify=((Item=(Type=Bone,Name="pack"),Transform=(Rotation=(X=-0.0066,Y=-0.0568,Z=0.1158,W=0.9916),Translation=(X=0,Y=0,Z=0),Scale3D=(X=1,Y=1,Z=1))))
    WeightMinimum=0
    Weight=@n1.OutputParam
   )
   > n2 = &RigUnit_SphericalPoseReader(
    ActiveRegionScaleFactors=(PositiveWidth=0.05,NegativeWidth=1,PositiveHeight=1,NegativeHeight=1)
    ActiveRegionSize=0.4
    Debug=(bDrawDebug=false,bDraw2D=False,bDrawLocalAxes=False,DebugScale=25,DebugSegments=20,DebugThickness=0.25)
    DriverAxis=(X=0,Y=-1,Z=0)
    DriverItem=(Type=Bone,Name="body_jnt_spine_05")
    FalloffRegionScaleFactors=(PositiveWidth=0.22,NegativeWidth=1,PositiveHeight=1,NegativeHeight=1)
    FalloffSize=0.5
    FlipHeightScaling=False
    FlipWidthScaling=False
    OptionalParentItem=(Type=Bone,Name="body_jnt_pelvis")
    RotationOffset=(X=-90,Y=-6,Z=-20)
   )
   > &RigUnit_ModifyTransforms(
    ItemToModify=((Item=(Type=Bone,Name="pack"),Transform=(Rotation=(X=0,Y=0.1428,Z=0,W=0.9898),Translation=(X=0,Y=0,Z=0),Scale3D=(X=1,Y=1,Z=1))))
    WeightMinimum=0
    Weight=@n2.OutputParam
   )
  ]}
]
hierarchy [
 root {loc=<0 0 0>, rot=<0 0 0 1>}
 pack {parent=root, loc=<0 -0.7316 0>, rot=<0 0 0 1>}
 Backpack_Hide {parent=pack, loc=<0 0 0>, rot=<0 0 0 1>}
 ribbon_top_01 {parent=pack, loc=<5.9329 -2.7771 2.7667>, rot=<0 0.5006 0 -0.8657>}
 ribbon_top_02 {parent=ribbon_top_01, loc=<3.2846 0 0>, rot=<0 0 0 1>}
 ribbon_bottom_01 {parent=pack, loc=<1.5195 -2.2341 -3.5366>, rot=<0 -0.9243 0 -0.3816>}
 ribbon_bottom_02 {parent=ribbon_bottom_01, loc=<3.2998 0 0>, rot=<0 0 0 1>}
 body_jnt_pelvis {parent=root, loc=<0 16.282 -36.7809>, rot=<0.0127 0.707 -0.0127 -0.707>}
 body_jnt_thigh_r {parent=body_jnt_pelvis, loc=<-2.5693 -0.0109 7.7501>, rot=<0.0572 -0.0012 0.9982 0.0155>}
 body_jnt_spine_03 {parent=body_jnt_pelvis, loc=<24.354 -2.6426 0>, rot=<0 0 0.058 -0.9983>}
 body_jnt_spine_05 {parent=body_jnt_pelvis, loc=<37.9129 -4.2232 0>, rot=<0 0 0.058 -0.9983>}
 body_jnt_lowerarm_r {parent=body_jnt_spine_05, loc=<-14.1803 -3.3706 37.6848>, rot=<0.8992 -0.1531 -0.3812 0.1508>}
 body_jnt_clavicle_l {parent=body_jnt_spine_05, loc=<4.3575 -0.1466 -2.3223>, rot=<0.0465 -0.7289 -0.009 -0.6829>}
 body_jnt_upperarm_l {parent=body_jnt_clavicle_l, loc=<15.5111 0 0>, rot=<0.0226 -0.3234 -0.0088 -0.9459>}
]
variables [SkyDiving {type=bool, default=False}]
