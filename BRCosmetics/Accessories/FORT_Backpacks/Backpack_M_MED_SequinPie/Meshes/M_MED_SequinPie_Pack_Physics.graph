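# Physics asset for the SequinPie backpack cosmetic, giving the pack collision shapes and limited swing relative to its root.
# A long, thin capsule on the pack bone is constrained to a tiny root sphere with limited swing and twist.
# Capsules on the pelvis, two spine joints, left clavicle and upper arm, and right thigh approximate only the parts of the wearer's body along that diagonal.
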PhysicsAsset /BRCosmetics/Accessories/FORT_Backpacks/Backpack_M_MED_SequinPie/Meshes/M_MED_SequinPie_Pack_Physics
bodies {
 body_jnt_clavicle_l [{shapeType=Capsule, length=9, radius=8, shapeName=body_jnt_clavicle_l_capsule, center=<8.0247 0 0>, rotation=<rot -90 92.894 -98.7836>}]
 body_jnt_pelvis [
  {shapeType=Capsule, length=7, radius=16, shapeName=body_jnt_pelvis_capsule, center=<-4.7725 2.48 0>, rotation=<rot 0 -90 0>}
  {shapeType=Capsule, length=8.5, radius=11, shapeName=body_jnt_pelvis_capsule, center=<8.3847 -0.3006 0>}
 ]
 body_jnt_spine_03 [{shapeType=Capsule, length=12.5, radius=10, shapeName=body_jnt_spine_03_capsule, center=<0 2.2717 0>, rotation=<rot 0 -90 0>}]
 body_jnt_spine_05 [{shapeType=Capsule, length=11.5, radius=7.5, shapeName=body_jnt_spine_05_capsule, center=<-1.048 2.5563 0>, rotation=<rot 0 -90 0>}]
 body_jnt_thigh_r [{shapeType=Capsule, length=15.5, radius=11, shapeName=body_jnt_thigh_r_capsule, center=<18.6167 0 0>, rotation=<rot 0 90 90>}]
 body_jnt_upperarm_l [{shapeType=Capsule, length=23, radius=8, shapeName=body_jnt_upperarm_l_capsule, center=<15.5902 0 0>, rotation=<rot -90 0 0>}]
 pack [{shapeType=Capsule, length=80, radius=3.0873, shapeName=pack_capsule, center=<1 -2 -7>, rotation=<rot -3.1119 -90 37.72>}]
 root [{shapeType=Sphere, radius=0.505, shapeName=root_sphere}]
}
constraints [{bone1Name=pack, bone2Name=root, swing1LimitDegrees=20, swing2LimitDegrees=20, twistLimitDegrees=30}]
