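# Animation Blueprint for the SequinPie backpack cosmetic, adding secondary motion and pack poses.
# The event graph caches body mesh socket transforms, which drive matching world-space bones before a Control Rig taking the skydiving state.
# Chained AnimDynamics springs then swing the ribbons, and bool blends add pack bone offsets for skydiving and front end.
# RigidBody and pack AnimDynamics branches exist but never reach the output pose.
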AnimBlueprint /BRCosmetics/Accessories/FORT_Backpacks/Backpack_M_MED_SequinPie/Meshes/M_MED_SequinPie_Pack_AnimBP
target_skeleton=&/M_MED_SequinPie_Pack_Skeleton
graphs [
 AnimGraph {nodes [
   n1 = Anim Dynamics - Bone: pack(ComponentPose=Local To Component(), Alpha=1, ExternalForce="0,0,0") {
    angularDampingOverride=0.9999
    angularSpringConstant=8000
    bAngularSpring=true
    bOverrideAngularDamping=true
    bUseGravityOverride=true
    bUsePlanarLimit=false
    gravityScale=0.35
    simulationSpace=RootRelative
    alphaBoolBlend {blendOption=Linear}
    boundBone {boneName=pack}
    componentAppliedLinearAccClamp=<1000 1000 1000>
    componentLinearAccScale=<0.6 0.5 0.5>
    componentLinearVelScale=<0.6 0.5 0.5>
    gravityOverride {z=-980}
    physicsBodyDefinitions [{
      boundBone {boneName=pack}
      boxExtents=<20 20 50>
      constraintSetup {angularLimitsMax=<5 5 10>, angularLimitsMin=<-5 -30 -10>, angularTarget=<1 0 0>}
      localJointOffset=<0 0 -10>
     }]
   }
   Blend Poses by bool(BlendPose_0=Use cached pose 'RBPose'(), BlendPose_1=Component To Local(ComponentPose=@n1), bActiveValue=Get bRigidBodyEnabled()) {blendType=Linear}
   n2 = "Transform (Modify) Bone - Bone: pack"!(ComponentPose=Local To Component(LocalPose=Use cached pose 'FinalPose'()), Alpha, Translation="0,0,0", Rotation="8,0,4") -> (Pose) {
    rotationMode=BMM_Additive
    translationMode=BMM_Additive
    translationSpace=BCS_BoneSpace
    alphaBoolBlend {blendInTime=1, blendOption=Linear, blendOutTime=1}
    boneToModify {boneName=pack}
    rotation {pitch=8, roll=4}
   }
   Output Pose(Result=Blend Poses by bool(
     BlendPose_0=Component To Local(ComponentPose=@n2)
     BlendPose_1=Use cached pose 'FinalPose'()
     BlendTime_0=0.3
     BlendTime_1=0.3
     bActiveValue=Get bIsInFrontEnd()
    ) {blendType=Linear}) {layerGroup=None}
   n3 = RigidBody!(ComponentPose=Local To Component(LocalPose=Use cached pose 'StartPose'()), Alpha) -> (Pose) {
    bClampLinearTranslationLimitToRefPose=true
    overridePhysicsAsset=&/M_MED_SequinPie_Pack_Physics
    simulationSpace=BaseBoneSpace
    alphaBoolBlend {blendOption=Linear}
    baseBoneRef {boneName=root}
    componentAppliedLinearAccClamp=<3000 3000 3000>
    componentLinearAccScale=<0.1 0.1 0.1>
    componentLinearVelScale=<0.1 0.1 0.2>
    overrideWorldGravity {z=-980}
   }
   n4 = RigidBody!(ComponentPose=Local To Component(LocalPose=Use cached pose 'StartPose'()), Alpha, OverrideWorldGravity:Vector="0,0,-980") -> (Pose) {
    bClampLinearTranslationLimitToRefPose=true
    bOverrideWorldGravity=true
    overridePhysicsAsset=&M_MED_SequinPie_Pack_Physics
    simulationSpace=BaseBoneSpace
    alphaBoolBlend {blendOption=Linear}
    baseBoneRef {boneName=root}
    componentAppliedLinearAccClamp=<3000 3000 3000>
    componentLinearAccScale=<0.1 0.1 0.1>
    componentLinearVelScale=<0.1 0.1 0.2>
    overrideWorldGravity {z=-980}
   }
   Save cached pose 'RBPose'(Pose=Blend Poses by bool(BlendPose_0=Component To Local(ComponentPose=@n4), BlendPose_1=Component To Local(ComponentPose=@n3), bActiveValue=Get bIsSkyDiving()) {blendType=Linear}) {cacheName=RBPose}
   n5 = "Transform (Modify) Bone - Bone: body_jnt_pelvis"!(
    ComponentPose=Local To Component(LocalPose=Input Pose())
    Alpha=1
    Translation=Get Pelvis_Location() -> (Pelvis_Location:Vector)
    Rotation=Get Pelvis_Rotation() -> (Pelvis_Rotation:Rotator)
   ) -> (Pose) {
    rotationMode=BMM_Replace
    rotationSpace=BCS_WorldSpace
    translationMode=BMM_Replace
    translationSpace=BCS_WorldSpace
    alphaBoolBlend {blendInTime=1, blendOption=Linear, blendOutTime=1}
    boneToModify {boneName=body_jnt_pelvis}
   }
   n6 = "Transform (Modify) Bone - Bone: body_jnt_thigh_r"!(
    ComponentPose=@n5
    Alpha=1
    Translation=Get Thigh_R_Location() -> (Thigh_R_Location:Vector)
    Rotation=Get Thigh_R_Rotation() -> (Thigh_R_Rotation:Rotator)
   ) -> (Pose) {
    rotationMode=BMM_Replace
    rotationSpace=BCS_WorldSpace
    translationMode=BMM_Replace
    translationSpace=BCS_WorldSpace
    alphaBoolBlend {blendInTime=1, blendOption=Linear, blendOutTime=1}
    boneToModify {boneName=body_jnt_thigh_r}
   }
   n7 = "Transform (Modify) Bone - Bone: body_jnt_spine_03"!(
    ComponentPose=@n6
    Alpha=1
    Translation=Get Spine_03_Location() -> (Spine_03_Location:Vector)
    Rotation=Get Spine_03_Rotation() -> (Spine_03_Rotation:Rotator)
   ) -> (Pose) {
    rotationMode=BMM_Replace
    rotationSpace=BCS_WorldSpace
    translationMode=BMM_Replace
    translationSpace=BCS_WorldSpace
    alphaBoolBlend {blendInTime=1, blendOption=Linear, blendOutTime=1}
    boneToModify {boneName=body_jnt_spine_03}
   }
   n8 = "Transform (Modify) Bone - Bone: body_jnt_spine_05"!(
    ComponentPose=@n7
    Alpha=1
    Translation=Get Spine_05_Location() -> (Spine_05_Location:Vector)
    Rotation=Get Spine_05_Rotation() -> (Spine_05_Rotation:Rotator)
   ) -> (Pose) {
    rotationMode=BMM_Replace
    rotationSpace=BCS_WorldSpace
    translationMode=BMM_Replace
    translationSpace=BCS_WorldSpace
    alphaBoolBlend {blendInTime=1, blendOption=Linear, blendOutTime=1}
    boneToModify {boneName=body_jnt_spine_05}
   }
   n9 = "Transform (Modify) Bone - Bone: body_jnt_clavicle_l"!(
    ComponentPose=@n8
    Alpha=1
    Translation=Get Clavicle_L_Location() -> (Clavicle_L_Location:Vector)
    Rotation=Get Clavicle_L_Rotation() -> (Clavicle_L_Rotation:Rotator)
   ) -> (Pose) {
    rotationMode=BMM_Replace
    rotationSpace=BCS_WorldSpace
    translationMode=BMM_Replace
    translationSpace=BCS_WorldSpace
    alphaBoolBlend {blendInTime=1, blendOption=Linear, blendOutTime=1}
    boneToModify {boneName=body_jnt_clavicle_l}
   }
   n10 = "Transform (Modify) Bone - Bone: body_jnt_upperarm_l"!(
    ComponentPose=@n9
    Alpha=1
    Translation=Get Upperarm_L_Location() -> (Upperarm_L_Location:Vector)
    Rotation=Get Upperarm_L_Rotation() -> (Upperarm_L_Rotation:Rotator)
   ) -> (Pose) {
    rotationMode=BMM_Replace
    rotationSpace=BCS_WorldSpace
    translationMode=BMM_Replace
    translationSpace=BCS_WorldSpace
    alphaBoolBlend {blendInTime=1, blendOption=Linear, blendOutTime=1}
    boneToModify {boneName=body_jnt_upperarm_l}
   }
   n11 = "Transform (Modify) Bone - Bone: body_jnt_lowerarm_r"!(
    ComponentPose=@n10
    Alpha=1
    Translation=Get Loverarm_R_Location() -> (Loverarm_R_Location:Vector)
    Rotation=Get Loverarm_R_Rotation() -> (Loverarm_R_Rotation:Rotator)
   ) -> (Pose) {
    rotationMode=BMM_Replace
    rotationSpace=BCS_WorldSpace
    translationMode=BMM_Replace
    translationSpace=BCS_WorldSpace
    alphaBoolBlend {blendInTime=1, blendOption=Linear, blendOutTime=1}
    boneToModify {boneName=body_jnt_lowerarm_r}
   }
   n12 = Control Rig!(Source=Component To Local(ComponentPose=@n11), Alpha, SkyDiving:bool=Get bIsSkyDiving()) -> (Pose) {
    alphaBoolBlend {blendOption=Linear}
    controlRigAssetReference {blueprintRigClass=&/M_MED_SequinPie_Pack_CtrlRig.M_MED_SequinPie_Pack_CtrlRig_C}
   }
   Save cached pose 'StartPose'(Pose=@n12) {cacheName=StartPose}
   n13 = "Transform (Modify) Bone - Bone: pack"(ComponentPose=Local To Component(LocalPose=Use cached pose 'AnimDynPose'()), Translation="-2.7,0,5.6", Rotation="-14,0,0") {
    rotationMode=BMM_Additive
    translationMode=BMM_Additive
    translationSpace=BCS_BoneSpace
    alphaBoolBlend {blendInTime=1, blendOption=Linear, blendOutTime=1}
    boneToModify {boneName=pack}
    rotation {pitch=-14}
    translation {x=-2.7, z=5.6}
   }
   Save cached pose 'FinalPose'(Pose=Blend Poses by bool(
     BlendPose_0=Component To Local(ComponentPose=@n13)
     BlendPose_1=Use cached pose 'AnimDynPose'()
     BlendTime_0=0.3
     BlendTime_1=0.3
     bActiveValue=Get bIsSkyDiving()
    ) {blendType=Linear}) {cacheName=FinalPose}
   n14 = Anim Dynamics - Bone: ribbon_top_01(ComponentPose=Local To Component(LocalPose=Use cached pose 'StartPose'()), Alpha=1, ExternalForce="0,0,0") {
    angularDampingOverride=0.9999
    angularSpringConstant=8000
    bAngularSpring=true
    bOverrideAngularDamping=true
    bUseGravityOverride=true
    bUsePlanarLimit=false
    gravityScale=0.35
    simulationSpace=RootRelative
    alphaBoolBlend {blendOption=Linear}
    boundBone {boneName=ribbon_top_01}
    componentAppliedLinearAccClamp=<1000 1000 1000>
    componentLinearAccScale=<0.6 0.5 0.5>
    componentLinearVelScale=<0.6 0.5 0.5>
    gravityOverride {z=-980}
    physicsBodyDefinitions [{
      boundBone {boneName=ribbon_top_01}
      boxExtents=<20 20 20>
      constraintSetup {angularLimitsMax=<0 1 10>, angularLimitsMin=<0 -30 -10>, angularTarget=<1 0 0>}
      localJointOffset=<1 0 1>
     }]
   }
   n15 = Anim Dynamics - Bone: ribbon_top_02(ComponentPose=@n14, Alpha=1, ExternalForce="0,0,0") {
    angularDampingOverride=0.9999
    angularSpringConstant=8000
    bAngularSpring=true
    bOverrideAngularDamping=true
    bUseGravityOverride=true
    bUsePlanarLimit=false
    gravityScale=0.35
    simulationSpace=RootRelative
    alphaBoolBlend {blendOption=Linear}
    boundBone {boneName=ribbon_top_02}
    componentAppliedLinearAccClamp=<1000 1000 1000>
    componentLinearAccScale=<0.6 0.5 0.5>
    componentLinearVelScale=<0.6 0.5 0.5>
    gravityOverride {z=-980}
    physicsBodyDefinitions [{
      boundBone {boneName=ribbon_top_02}
      boxExtents=<15 20 20>
      constraintSetup {angularLimitsMax=<0 1 10>, angularLimitsMin=<0 -20 -10>, angularTarget=<1 0 0>}
      localJointOffset=<1 0 0.5>
     }]
   }
   n16 = Anim Dynamics - Bone: ribbon_bottom_01(ComponentPose=@n15, Alpha=1, ExternalForce="0,0,0") {
    angularDampingOverride=0.9999
    angularSpringConstant=8000
    bAngularSpring=true
    bOverrideAngularDamping=true
    bUseGravityOverride=true
    bUsePlanarLimit=false
    gravityScale=0.35
    simulationSpace=RootRelative
    alphaBoolBlend {blendOption=Linear}
    boundBone {boneName=ribbon_bottom_01}
    componentAppliedLinearAccClamp=<1000 1000 1000>
    componentLinearAccScale=<0.6 0.5 0.5>
    componentLinearVelScale=<0.6 0.5 0.5>
    gravityOverride {z=-980}
    physicsBodyDefinitions [{
      boundBone {boneName=ribbon_bottom_01}
      boxExtents=<25 20 20>
      constraintSetup {angularLimitsMax=<0 20 10>, angularLimitsMin=<0 -5 -10>, angularTarget=<1 0 0>}
      localJointOffset=<1.5 0 1>
     }]
   }
   n17 = Anim Dynamics - Bone: ribbon_bottom_02(ComponentPose=@n16, Alpha=1, ExternalForce="0,0,0") {
    angularDampingOverride=0.9999
    angularSpringConstant=8000
    bAngularSpring=true
    bOverrideAngularDamping=true
    bUseGravityOverride=true
    bUsePlanarLimit=false
    gravityScale=0.35
    simulationSpace=RootRelative
    alphaBoolBlend {blendOption=Linear}
    boundBone {boneName=ribbon_bottom_02}
    componentAppliedLinearAccClamp=<1000 1000 1000>
    componentLinearAccScale=<0.6 0.5 0.5>
    componentLinearVelScale=<0.6 0.5 0.5>
    gravityOverride {z=-980}
    physicsBodyDefinitions [{
      boundBone {boneName=ribbon_bottom_02}
      boxExtents=<15 20 20>
      constraintSetup {angularLimitsMax=<0 15 10>, angularLimitsMin=<0 -2 -10>, angularTarget=<1 0 0>}
      localJointOffset=<0.5 0 0.5>
     }]
   }
   Save cached pose 'AnimDynPose'(Pose=Component To Local(ComponentPose=@n17)) {cacheName=AnimDynPose}
  ]}
 EventGraph {nodes [
   Event BlueprintUpdateAnimation()
   n1 = Get BodyMesh() -> (BodyMesh:Skeletal Mesh Component Object Reference)
   > Is Valid(InputObject=@n1)
   n2 = BreakTransform(InTransform=GetSocketTransform(self=@n1, InSocketName=pelvis, TransformSpace=RTS_World))
   > Set Pelvis_Location(Pelvis_Location:Vector=@n2) -> (Output_Get:Vector)
   > Set Pelvis_Rotation(Pelvis_Rotation:Rotator=@n2.Rotation) -> (Output_Get:Rotator)
   n3 = BreakTransform(InTransform=GetSocketTransform(self=@n1, InSocketName=spine_05, TransformSpace=RTS_World))
   > Set Spine_05_Location(Spine_05_Location:Vector=@n3) -> (Output_Get:Vector)
   > Set Spine_05_Rotation(Spine_05_Rotation:Rotator=@n3.Rotation) -> (Output_Get:Rotator)
   n4 = BreakTransform(InTransform=GetSocketTransform(self=@n1, InSocketName=thigh_r, TransformSpace=RTS_World))
   > Set Thigh_R_Location(Thigh_R_Location:Vector=@n4) -> (Output_Get:Vector)
   > Set Thigh_R_Rotation(Thigh_R_Rotation:Rotator=@n4.Rotation) -> (Output_Get:Rotator)
   n5 = BreakTransform(InTransform=GetSocketTransform(self=@n1, InSocketName=lowerarm_r, TransformSpace=RTS_World))
   > Set Loverarm_R_Location(Loverarm_R_Location:Vector=@n5) -> (Output_Get:Vector)
   > Set Loverarm_R_Rotation(Loverarm_R_Rotation:Rotator=@n5.Rotation) -> (Output_Get:Rotator)
   n6 = Get BodyMesh()
   n7 = BreakTransform(InTransform=GetSocketTransform(self=@n6, InSocketName=upperarm_l, TransformSpace=RTS_World))
   > Set Upperarm_L_Location(Upperarm_L_Location:Vector=@n7) -> (Output_Get:Vector)
   > Set Upperarm_L_Rotation(Upperarm_L_Rotation:Rotator=@n7.Rotation) -> (Output_Get:Rotator)
   n8 = BreakTransform(InTransform=GetSocketTransform(self=@n6, InSocketName=spine_03, TransformSpace=RTS_World))
   > Set Spine_03_Location(Spine_03_Location:Vector=@n8) -> (Output_Get:Vector)
   > Set Spine_03_Rotation(Spine_03_Rotation:Rotator=@n8.Rotation) -> (Output_Get:Rotator)
   n9 = BreakTransform(InTransform=GetSocketTransform(self=@n6, InSocketName=clavicle_l, TransformSpace=RTS_World))
   > Set Clavicle_L_Location(Clavicle_L_Location:Vector=@n9) -> (Output_Get:Vector)
   > Set Clavicle_L_Rotation(Clavicle_L_Rotation:Rotator=@n9.Rotation) -> (Output_Get:Rotator)
   n10 = Event OnCharacterCustomizationLoaded()
   > Set BodyMesh(BodyMesh=Get Array Item with EFort Custom Part Type(Array=@n10.CustomMeshParts))
   Reroute Node()
   > n11 = Gravity Override State Function DEPRECATED(CharacterPartAnimInstance=Self-Reference(), GameTimeInSeconds=GetGameTimeInSeconds(), GravityOverrideParams=Get Gravity Override Params())
   > Set Gravity Override(Gravity Override=@n11)
  ]}
]
variables [
 BodyMesh {type=SkeletalMeshComponent}
 Clavicle_L_Location {type=Vector}
 Clavicle_L_Rotation {type=Rotator}
 Gravity Override {type=Vector}
 Gravity Override Params {type=GravityOverrideParamsStruct}
 Loverarm_R_Location {type=Vector}
 Loverarm_R_Rotation {type=Rotator}
 Pelvis_Location {type=Vector}
 Pelvis_Rotation {type=Rotator}
 Spine_03_Location {type=Vector}
 Spine_03_Rotation {type=Rotator}
 Spine_05_Location {type=Vector}
 Spine_05_Rotation {type=Rotator}
 Thigh_R_Location {type=Vector}
 Thigh_R_Rotation {type=Rotator}
 Upperarm_L_Location {type=Vector}
 Upperarm_L_Rotation {type=Rotator}
]
